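# Additive Niagara mesh-particle material for a boiling water wave ring in the Flopper upgrade effect.
# Radially panned rolling textures, masked to a soft ring, drive normal-based vertex bulging, added to water-surface displacement and camera offsets.
# A depth-graded two-color tint, scaled by view angle and distance, pairs with GGX glints and Fresnel sky reflections.
# Opacity combines depth and game-time fades; platform switches zero color and displacement outside high-end PC and current-gen consoles.
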Material /Game/Effects/Fort_Effects/Gameplay/Floppers/FlopperUpdrade/M_BoilingWave
graph {
 properties {bUsedWithNiagaraMeshParticles=true, blendMode=BLEND_Additive, usedShadingModels=Default Lit}
 nodes [
  n1 = NamedRerouteUsage()
  n2 = ScalarParameter() {parameterName=TmpDisableGameTimeFade}
  n3 = Multiply(A=DotProduct(
    A=VertexNormalWS()
    B=MaterialFunctionCall() {materialFunction=&CameraDirectionVector}.Result
   )) {constB=-1}
  n4 = LinearInterpolate(
   A=ScalarParameter() {defaultValue=130, parameterName=DepthFadeDistance}
   B=ScalarParameter() {defaultValue=15, parameterName=DepthFadeDistanceB}
   Alpha=@n3
  ) {constB=200}
  n5 = TextureCoordinate()
  n6 = Constant() {r=0.5}
  n7 = Multiply(
   A=SphereMask(A=TextureCoordinate(), B=Constant() {r=0.5}) {attenuationRadius=0.5, hardnessPercent=0}
   B=OneMinus(None=SphereMask(A=@n5, B=@n6) {attenuationRadius=0.2, hardnessPercent=0.5})
  )
  n8 = VectorParameter() {parameterName=WaveUV, defaultValue=<rgba 0 -0.8 1 1.5>}
  n9 = Panner(
   Coordinate=Multiply(
    A=MaterialFunctionCall() {materialFunction=&VectorToRadialValue}.Radial Coordinates
    B=AppendVector(A=@n8.B, B=@n8.B)
   )
   Time=Time() {bOverride_Period=true, period=6000}
   Speed=AppendVector(A=@n8.R, B=@n8.R)
  )
  n10 = TextureSample(UVs=@n9) {samplerType=SAMPLERTYPE_LinearColor, texture=&/T_RollingTexture_G}
  n11 = Multiply(A=@n7, B=Multiply(A=@n7, B=Power(Base=@n10.RGB)))
  n12 = QualitySwitch(Default=ShadingPathSwitch(Default=DepthFade(Opacity=@n7, FadeDistance=@n4), Forward=@n11, Mobile=@n11), Low=@n11)
  n13 = ScalarParameter() {defaultValue=1, parameterName=FinalOpacity}
  DepthFade(Opacity=Multiply(A=LinearInterpolate(A=@n1, Alpha=@n2), B=Multiply(A=@n12, B=@n13))) {fadeDistanceDefault=2}
  n14 = ParticlePositionWS()
  n15 = ObjectPositionWS()
  FeatureLevelSwitch(Default=@n14, ES2=@n15, ES3_1=@n15, SM4_REMOVED=@n14)
  n16 = Constant()
  n17 = TextureSample(UVs=@n9) {samplerType=SAMPLERTYPE_Normal, texture=&/T_RolingTexture}
  n18 = Transform(None=@n17.RGB)
  n19 = Multiply(
   A=MaterialFunctionCall!(LightDirctionIn, CameraVectorIn, Roughness=ScalarParameter() {defaultValue=0.05, parameterName=Roughness}, NormalIn=@n18) {materialFunction=/Game/Effects/Niagara/MasterMaterials/Functions/Common/FXCommon_GGX}.Result
   B=VectorParameter() {parameterName=GGXColor, defaultValue=<rgba 1 0.7393 0.5 1>}.RGB
  )
  n20 = VectorParameter() {parameterName=FresnelParameter, defaultValue=<rgba 1 0 0 1>}
  n21 = Transform(None=MaterialFunctionCall(Normal=@n17.RGB, Flatness=ScalarParameter() {defaultValue=-2, parameterName=FlatNormal}) {materialFunction=&FlattenNormal}.Result)
  n22 = Multiply(
   A=SkyLightEnvMapSample(Direction=ReflectionVectorWS(CustomWorldNormal=@n21), Roughness=Constant())
   B=VectorParameter() {parameterName=ReflectionColor, defaultValue=<rgba 0.25 0.25 0.25 1>}.RGB
  )
  n23 = Constant()
  n24 = Multiply(
   A=Fresnel(ExponentIn=@n20.R, BaseReflectFractionIn=@n20.R, Normal=@n18)
   B=QualitySwitch(Default=ShadingPathSwitch(Default=@n22, Forward=@n23, Mobile=@n23), Low=@n23)
  )
  n25 = ScalarParameter() {defaultValue=1, group=Emissive, parameterName=Emissive Night Multiplier}
  n26 = CollectionParameter() {
   collection=&FortniteMaterialParameters
   parameterName=FortniteActiveTimeOfDay
  }
  n27 = ComponentMask(None=@n26) {a=true}
  MaterialFunctionCall(
   "Xb1,PS4"=@n16
   "PC-High&&Epic,XSX,PS5"=Multiply(A=Add(A=@n19, B=@n24), B=LinearInterpolate(B=@n25, Alpha=@n27) {constA=1})
   "PC-Perf,Android"=@n16
   Switch=@n16
   "PC-Low(NotPerf)&&Medium"=@n16
  ) {materialFunction=&PlatformSwitch_1D}
  n28 = Constant()
  n29 = VectorParameter() {parameterName=ZOffsetByCameraDistance, defaultValue=<rgba 1000 5000 0 50>}
  n30 = Subtract(A=Distance(A=CameraPositionWS(), B=ActorPositionWS()), B=@n29.R)
  n31 = LinearInterpolate(A=@n29.B, B=@n29.B, Alpha=Saturate(None=Divide(A=@n30, B=@n29.G)))
  n32 = VertexNormalWS()
  n33 = Multiply(A=@n7, B=Multiply(A=Multiply(A=@n10.RGB, B=@n32), B=ScalarParameter() {defaultValue=75, parameterName=WPOStr}))
  n34 = VectorParameter() {parameterName=WPOOffset, defaultValue=<rgba 0 0 -30 1>}
  n35 = Constant()
  n36 = QualitySwitch(Default=ShadingPathSwitch(Default=Add(A=@n33, B=@n34.RGB), Forward=@n35, Mobile=@n35), Low=@n35)
  n37 = StaticBoolParameter() {defaultValue=true, parameterName=Ignore XY Gerstner Offsets}
  n38 = ScalarParameter() {parameterName=WaterBodyIndex}
  n39 = ScalarParameter() {parameterName=WaterDepth}
  n40 = ScalarParameter() {defaultValue=5, parameterName=ZOffset}
  n41 = CameraPositionWS()
  n42 = ActorPositionWS()
  n43 = Clamp(None=ComponentMask(None=Subtract(A=@n41, B=@n42)) {b=true}) {minDefault=-1}
  n44 = Add(
   A=Multiply(
    A=MaterialFunctionCall!(Ignore XY Offsets=@n37, Position, WaterBodyIndex=@n38, Depth=@n39) {materialFunction=/FortWater/Materials/Functions/WaterParticleDisplacement}.WPO
    B=VectorParameter() {parameterName=Water Displacement Multiplier, defaultValue=<rgba 1 1 1 0>}.RGB
   )
   B=MaterialFunctionCall(Z=Multiply(A=@n40, B=@n43)) {materialFunction=&MakeFloat3}.Result
  )
  n45 = Constant()
  n46 = Add(
   A=MaterialFunctionCall(Offset Amount=ScalarParameter() {defaultValue=10, parameterName=CameraOffset}) {materialFunction=/Engine/Functions/Engine_MaterialFunctions02/WorldPositionOffset/CameraOffset}.Clamped Camera Offset
   B=LinearInterpolate(
    A=StaticSwitchParameter(True=@n44, False=@n45) {defaultValue=true, parameterName=UseWaterDisplacement}
    Alpha=ScalarParameter() {parameterName=TmpDisableWaterDisplacement}
   ) {constB=0}
  )
  MaterialFunctionCall(
   "Xb1,PS4"=@n28
   "PC-High&&Epic,XSX,PS5"=Add(
    A=MaterialFunctionCall(Z=@n31) {materialFunction=&MakeFloat3}.Result
    B=Add(A=@n36, B=@n46)
   )
   "PC-Perf,Android"=@n28
   Switch=@n28
   "PC-Low(NotPerf)&&Medium"=@n28
  ) {materialFunction=&PlatformSwitch_1D}
  MaterialFunctionCall("Xb1,PS4"=@n28, "PC-High&&Epic,XSX,PS5"=@n28, "PC-Perf,Android"=@n28, Switch=@n28, "PC-Low(NotPerf)&&Medium"=@n28) {materialFunction=&PlatformSwitch_1D}
  n47 = LinearInterpolate(
   A=VectorParameter() {parameterName=ColorA, defaultValue=<rgba 0 0.1 0.15 1>}.RGB
   B=VectorParameter() {parameterName=ColorB, defaultValue=<rgba 0 0.25 0.25 1>}.RGB
   Alpha=DepthFade(FadeDistance=ScalarParameter() {defaultValue=200, parameterName=ColorGradientDepthFade})
  )
  n48 = Multiply(A=@n47, B=ScalarParameter() {defaultValue=2, parameterName=LowpatformScaleColor})
  n49 = ShadingPathSwitch(Default=@n47, Forward=@n48, Mobile=@n48)
  n50 = Constant() {r=1}
  n51 = QualitySwitch(Default=ShadingPathSwitch(Default=Constant() {r=0.75}, Forward=@n50, Mobile=@n50), Low=@n50)
  n52 = Subtract(A=Distance(A=CameraPositionWS(), B=ActorPositionWS())) {constB=4000}
  n53 = LinearInterpolate(
   A=ScalarParameter() {defaultValue=0.5, parameterName=ScaleOpacity_SideAngle}
   B=ScalarParameter() {defaultValue=1.25, parameterName=ScaleOpacity_TopAngle}
   Alpha=Saturate(None=Divide(A=@n52) {constB=3000})
  )
  MaterialFunctionCall(
   "Xb1,PS4"=@n16
   "PC-High&&Epic,XSX,PS5"=Multiply(A=QualitySwitch(Default=@n49, Low=@n48), B=LinearInterpolate(A=@n51, B=@n53, Alpha=@n3) {constA=0.7})
   "PC-Perf,Android"=@n16
   Switch=@n16
   "PC-Low(NotPerf)&&Medium"=@n16
  ) {materialFunction=&PlatformSwitch_1D}
  n54 = Time()
  n55 = ScalarParameter() {group=Game Time Fade, parameterName=FadeGameTime}
  n56 = Saturate(None=Divide(A=Subtract(A=@n54, B=@n55), B=ScalarParameter() {defaultValue=2, group=Game Time Fade, parameterName=FadeDuration}))
  NamedRerouteDeclaration(None=LinearInterpolate(A=OneMinus(None=@n56), B=@n56, Alpha=ScalarParameter() {defaultValue=1, group=Game Time Fade, parameterName=FadeDirection})) {name=Game Time Fade, nodeColor=<rgba 1 0 0.9141 1>}
 ]
}
