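# Lit flopper fish material for Niagara mesh particles, with dynamic parameters driving animation, hue, desaturation, darkening and shine.
# World position offset adds a panned-gradient swing, faded along local Y, to a particle-pivot twist whose axis reads only RotX.
# Base color blends the tinted body with a UV-masked eye, then brightens sun-facing sides.
# A GGX sun highlight over tiled crystal normals supplies the emissive skin glint.
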Material /Game/Effects/Fort_Effects/Gameplay/Floppers/FlopperUpdrade/M_NewFopperMaterial
graph {
 properties {bUsedWithNiagaraMeshParticles=true, pixelDepthOffsetMode=PDOM_Legacy, usedShadingModels=Default Lit}
 nodes [
  n1 = Divide(A=MaterialFunctionCall(Float3=MaterialFunctionCall() {materialFunction=&LocalPosition}."Instance Local Position (Excluding Offsets)") {materialFunction=&BreakOutFloat3Components}.G) {constB=32}
  n2 = Add(A=@n1) {constB=0.5}
  n3 = ComponentMask(None=VectorParameter() {group=WPO, parameterName=TextureWPO_UVtiling, defaultValue=<rgba 0.4 0 0 0>}.RGB) {g=true, r=true}
  n4 = DynamicParameter() {paramNames [Time,HueShift,SwingStr,Desat]}
  n5 = VectorParameter() {group=WPO, parameterName=TextureWPO_Speed, defaultValue=<rgba 1 0 0 0>}
  n6 = ComponentMask(None=@n5.RGB) {r=true}
  n7 = ComponentMask(None=@n5.RGB) {g=true}
  n8 = TextureSample(UVs=Panner(Coordinate=Multiply(A=@n2, B=@n3), Time=@n4.Time, Speed=AppendVector(A=@n6, B=@n7)) {speedX=0.15}) {texture=/Game/Global/Textures/Default/Masks_and_Gradients/WeaponGradients/T_Gradient_DesertSky_Color}
  n9 = Multiply(A=@n8.R, B=ScalarParameter() {defaultValue=8, group=WPO, parameterName=WPO_HeightX}) {constB=0.5}
  n10 = Multiply(A=@n8.R, B=ScalarParameter() {defaultValue=3, group=WPO, parameterName=WPO_HeightY}) {constB=0.5}
  n11 = ScalarParameter() {defaultValue=1, group=WPO, parameterName=WPO_HeightZ}
  n12 = Multiply(A=AppendVector(A=AppendVector(A=@n9, B=@n10), B=Multiply(A=@n8.R, B=@n11) {constB=0.6}), B=@n4.SwingStr)
  n13 = OneMinus(None=Power(Base=@n2, Exp=ScalarParameter() {defaultValue=1, parameterName=WPO_GradientPower}))
  n14 = DynamicParameter() {parameterIndex=3, paramNames [RotX,RotY,RotZ,RotAngle]}
  Add(
   A=Transform(None=Multiply(A=@n12, B=@n13)) {transformSourceType=TRANSFORMSOURCE_Local}
   B=RotateAboutAxis(
    NormalizedRotationAxis=MaterialFunctionCall(R=@n14.RotX, G=@n14.RotX, B=@n14.RotX) {materialFunction=&AppendMany}.RGB
    RotationAngle=Multiply(A=@n14.RotAngle, B=@n1)
    PivotPoint=ParticlePositionWS()
    Position=WorldPosition().XYZ
   )
  )
  n15 = TextureSample() {samplerType=SAMPLERTYPE_Normal, texture=/Game/Gadgets/Assets/Fish/Less_Effective_Flopper_Mark_II/Textures/T_Item_Less_Effective_Flopper_N}
  MaterialFunctionCall(Normal=@n15.RGB, Flatness=ScalarParameter() {defaultValue=-0.3, parameterName=FlatFishNormal}) {materialFunction=&FlattenNormal}
  n16 = SkyAtmosphereLightDirection()
  n17 = ScalarParameter() {defaultValue=0.03, parameterName=FishHighLightRou}
  n18 = Multiply(A=TextureCoordinate(), B=ScalarParameter() {defaultValue=8, parameterName=FishSkinNormalUVDensity})
  n19 = Transform(None=MaterialFunctionCall(
    Normal=TextureSample(UVs=@n18) {samplerType=SAMPLERTYPE_Normal, texture=/Game/FrontEnd/Store/Cards/Assets/T_Crystal_N}.RGB
    Flatness=ScalarParameter() {defaultValue=0.5, parameterName=FishSkinNormalFlat}
   ) {materialFunction=&FlattenNormal}.Result)
  n20 = Multiply(
   A=ScalarParameter() {defaultValue=400, parameterName=FishHighLightAreaMul}
   B=Saturate(None=MaterialFunctionCall!(LightDirctionIn=@n16, CameraVectorIn, Roughness=@n17, NormalIn=@n19) {materialFunction=/Game/Effects/Niagara/MasterMaterials/Functions/Common/FXCommon_GGX}.Result)
  )
  n21 = Multiply(
   A=ScalarParameter() {defaultValue=4000, parameterName=FishHighLightStr}
   B=Power(Base=Saturate(None=@n20), Exp=ScalarParameter() {defaultValue=2.3, parameterName=FishHighLightAreaPower})
  )
  n22 = DynamicParameter() {parameterIndex=1, defaultValue=<rgba 0.8 1 1 1>, paramNames [Dark, Shining, "", ""]}
  MaterialFunctionCall(Emissive=Multiply(A=VectorParameter() {parameterName=FishHightLightColor, defaultValue=<rgba 1 0.7384 0.5 1>}.RGB, B=Multiply(A=@n21, B=@n22.Shining))) {materialFunction=&MF_FX_EmissiveAdjustments}
  n23 = TextureSample() {samplerType=SAMPLERTYPE_Masks, texture=/Game/Gadgets/Assets/Fish/Less_Effective_Flopper_Mark_II/Textures/T_Item_Less_Effective_Flopper_S}
  Multiply(A=@n23.R) {constB=0.2}
  Multiply(A=@n23.B) {constB=0.2}
  n24 = DynamicParameter() {paramNames [Time,HueShift,SwingStr,Desat]}
  n25 = TextureSampleParameter2D() {parameterName=FishT, texture=/Game/Gadgets/Assets/Fish/Less_Effective_Flopper_Mark_II/Textures/T_Item_Less_Effective_Flopper_ORIGINAL_D}
  n26 = Desaturation(
   None=MaterialFunctionCall(Hue Shift Percentage=@n24.HueShift, Texture=@n25.RGB) {materialFunction=&HueShift}.Result
   Fraction=@n24.Desat
  )
  n27 = DynamicParameter() {parameterIndex=1, defaultValue=<rgba 0.8 1 1 1>, paramNames [Dark, Shining, "", ""]}
  n28 = TextureSampleParameter2D(UVs=TextureCoordinate() {coordinateIndex=1}) {
   group=Global
   parameterName=EyeTexture
   samplerSource=SSM_Wrap_WorldGroupSettings
   texture=/Game/Gadgets/Assets/Fish/Flopper_Mark_II/Textures/T_Item_Flopper_Mark_II_Eye_D
  }
  n29 = VectorParameter() {parameterName=EyeColor, defaultValue=<rgba 1 1 1 1>}
  n30 = VectorParameter() {parameterName="EyeMaskPosition(Do Not Touch)", defaultValue=<rgba 0.5 -0.5 0 1>}
  n31 = MaterialFunctionCall(
   UVs=Add(A=TextureCoordinate(), B=ComponentMask(None=@n30.RGB) {g=true, r=true})
   Texture Scale=AppendVector(
    A=ScalarParameter() {defaultValue=0.1, parameterName="X_EyeMask_Size(Do Not Touch)"}
    B=ScalarParameter() {defaultValue=0.1, parameterName="Y_EyeMask_Size(Do Not Touch)"}
   )
  ) {materialFunction=&ScaleUVsByCenter}
  n32 = Saturate(None=DotProduct(A=@n16, B=PixelNormalWS()))
  n33 = ScalarParameter() {defaultValue=1, parameterName=FaceToSunSideBrightness}
  Multiply(
   A=LinearInterpolate(A=Multiply(A=@n26, B=@n27.Dark) {constB=0.5}, B=Multiply(A=@n28.RGB, B=@n29.RGB), Alpha=Clamp(None=@n31.0-1 mask))
   B=Add(A=Multiply(A=@n32, B=@n33) {constB=15}, B=ScalarParameter() {defaultValue=0.8, parameterName=NotFaceToSunSideBrightness}) {constB=0.5}
  )
 ]
}
