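# Translucent lit water ripple material for Niagara mesh particles in the Flopper upgrade gameplay effect.
# Radially mapped, panning normal, flow and foam textures build the rings, driving color, opacity and double scene-color refraction.
# A distorted glow with night boost, game-time fade and water-body color tint layer on top.
# World position offset follows water displacement plus camera offset, while quality and shading-path switches simplify refraction and lighting on Low, Mobile or Forward.
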Material /Game/Effects/Fort_Effects/Gameplay/Floppers/FlopperUpdrade/M_Ripple
graph {
 properties {
  bUseTranslucencyVertexFog=false
  bUsedWithNiagaraMeshParticles=true
  blendMode=BLEND_Translucent
  pixelDepthOffsetMode=PDOM_Legacy
  translucencyPass=MTP_BeforeDOF
  usedShadingModels=Default Lit
 }
 nodes [
  Constant3Vector() {constant=<rgba 0.27 0.27 0.27 1>}
  Constant() {r=1}
  DynamicParameter() {parameterIndex=3, defaultValue=<rgba 0 512 1 0>, paramNames [WaterBodyIndex, WaterDepth, WaveAttenuationFactor, Param4]}
  n1 = MaterialFunctionCall() {materialFunction=&VectorToRadialValue}
  n2 = MaterialFunctionCall(Float2=@n1.Radial Coordinates) {materialFunction=&BreakOutFloat2Components}
  NamedRerouteDeclaration(None=OneMinus(None=@n2.G)) {name=CircleMask, nodeColor=<rgba 1 0 0.0469 1>}
  n3 = VectorParameter() {group=Move, parameterName=Move2, defaultValue=<rgba 1 3 0 -1>}
  n4 = AppendVector(A=@n3.R, B=@n3.R)
  n5 = MaterialFunctionCall(Float2=ActorPositionWS()) {materialFunction=&BreakOutFloat2Components}
  n6 = Constant() {r=100}
  n7 = Add(
   A=Multiply(A=Time() {bOverride_Period=true, period=6000}, B=ScalarParameter() {defaultValue=1, group=Move, parameterName=Move Time Multiplier})
   B=LinearInterpolate(B=Fmod(A=@n5.G, B=@n6), Alpha=ScalarParameter() {defaultValue=1, parameterName=UseRandomUVCoordinate})
  )
  n8 = TextureSample(UVs=Panner(Coordinate=Multiply(A=@n1.Radial Coordinates, B=@n4), Time=@n7, Speed=AppendVector(A=@n3.B, B=@n3.B)) {speedY=-0.5}) {samplerType=SAMPLERTYPE_Normal, texture=&/T_PerlinNormal}
  n9 = SphereMask(A=TextureCoordinate(), B=Constant2Vector() {g=0.5, r=0.5}) {attenuationRadius=0.2, hardnessPercent=10}
  NamedRerouteDeclaration(None=Multiply(A=AppendVector(A=@n8.R, B=@n8.R), B=OneMinus(None=@n9))) {name=SphereGlowDistortion, nodeColor=<rgba 1 0 0.4453 1>}
  n10 = Time()
  n11 = ScalarParameter() {group=Game Time Fade, parameterName=FadeGameTime}
  n12 = Saturate(None=Divide(A=Subtract(A=@n10, B=@n11), B=ScalarParameter() {defaultValue=2, group=Game Time Fade, parameterName=FadeDuration}))
  NamedRerouteDeclaration(None=LinearInterpolate(A=OneMinus(None=@n12), B=@n12, Alpha=ScalarParameter() {defaultValue=1, group=Game Time Fade, parameterName=FadeDirection})) {name=Game Time Fade, nodeColor=<rgba 0.0312 1 0 1>}
  n13 = NamedRerouteUsage()
  n14 = Add(
   A=Power(Base=OneMinus(None=@n13)) {constExponent=5}
   B=SphereMask(A=TextureCoordinate(), B=Constant2Vector() {g=0.5, r=0.5}) {attenuationRadius=0.2, hardnessPercent=0}
  )
  n15 = ScalarParameter() {group=Emissive Sphere Glow, parameterName=Emissive Sphere Glow Multiplier}
  n16 = Multiply(A=NamedRerouteUsage(), B=ScalarParameter() {group=Emissive Sphere Glow, parameterName=Emissive Sphere Glow Distortion})
  n17 = TextureCoordinate()
  n18 = ScalarParameter() {defaultValue=0.5, group=Emissive Sphere Glow, parameterName=Emissive Sphere Glow Radius}
  n19 = Multiply(A=@n18) {constB=2}
  n20 = Power(
   Base=SphereMask(A=Add(A=@n16, B=@n17), B=Constant2Vector() {g=0.5, r=0.5}, Radius=QualitySwitch(Default=@n18, Low=@n19, Medium=@n19)) {hardnessPercent=0}
   Exp=ScalarParameter() {defaultValue=1, group=Emissive Sphere Glow, parameterName=Emissive Sphere Glow Power}
  )
  NamedRerouteDeclaration(None=Multiply(A=Saturate(None=@n14), B=Multiply(A=@n20, B=@n15))) {name=EmissiveSphereGlowMask, nodeColor=<rgba 1 0 0.0937 1>}
  n21 = VectorParameter() {group=Emissive Sphere Glow, parameterName=Emissive Sphere Glow Color, defaultValue=<rgba 0 0 0 1>}
  n22 = CollectionParameter() {
   collection=&FortniteMaterialParameters
   parameterName=FortniteActiveTimeOfDay
  }
  n23 = Multiply(
   A=Multiply(A=@n20, B=Multiply(A=@n15, B=@n21.RGB))
   B=LinearInterpolate(
    B=ScalarParameter() {defaultValue=1, group=Emissive Sphere Glow, parameterName=Emissive Sphere Glow Night Multiplier}
    Alpha=ComponentMask(None=@n22) {a=true}
   ) {constA=1}
  )
  NamedRerouteDeclaration(None=QualitySwitch(Default=@n23, Low=Multiply(A=@n23) {constB=5})) {name=EmissiveSphereGlow, nodeColor=<rgba 0.3125 0 1 1>}
  n24 = VectorParameter() {group=Move, parameterName=Move3, defaultValue=<rgba 1 1 0 -0.3>}
  n25 = Panner(Coordinate=Multiply(A=@n1.Radial Coordinates, B=AppendVector(A=@n24.R, B=@n24.R)), Time=@n7, Speed=AppendVector(A=@n24.B, B=@n24.B)) {speedY=-0.5}
  n26 = TextureCoordinate()
  n27 = DDX(Value=@n26)
  n28 = DDY(Value=@n26)
  n29 = TextureSample!(UVs=@n25, Tex, "DDX(UVs)"=@n27, "DDY(UVs)"=@n28, Apply View MipBias) {
   mipValueMode=TMVM_Derivative
   samplerType=SAMPLERTYPE_LinearGrayscale
   texture=&/T_FlowSandTile
  }
  NamedRerouteDeclaration(None=@n29.R) {name=EmissiveSphereGlowRippleMask, nodeColor=<rgba 0.0781 0 1 1>}
  n30 = ScalarParameter() {defaultValue=10, parameterName=CameraOffset}
  n31 = StaticBoolParameter() {defaultValue=true, parameterName=Ignore XY Gerstner Offsets}
  n32 = ScalarParameter() {parameterName=WaterBodyIndex}
  n33 = ScalarParameter() {parameterName=WaterDepth}
  n34 = ScalarParameter() {defaultValue=5, parameterName=ZOffset}
  n35 = CameraPositionWS()
  n36 = ShadingPathSwitch(Default=ParticlePositionWS(), Mobile=ObjectPositionWS()) {desc=Migrated from FeatureLevelSwitch}
  n37 = Clamp(None=ComponentMask(None=Subtract(A=@n35, B=@n36)) {b=true}) {minDefault=-1}
  n38 = Add(
   A=Multiply(
    A=MaterialFunctionCall!(Ignore XY Offsets=@n31, Position, WaterBodyIndex=@n32, Depth=@n33) {materialFunction=/FortWater/Materials/Functions/WaterParticleDisplacement}.WPO
    B=VectorParameter() {parameterName=Water Displacement Multiplier, defaultValue=<rgba 1 1 1 0>}.RGB
   )
   B=MaterialFunctionCall(Z=Multiply(A=@n34, B=@n37)) {materialFunction=&MakeFloat3}.Result
  )
  n39 = Constant()
  Reroute(None=Add(
    A=MaterialFunctionCall(Offset Amount=@n30) {materialFunction=/Engine/Functions/Engine_MaterialFunctions02/WorldPositionOffset/CameraOffset}.Clamped Camera Offset
    B=StaticSwitchParameter(True=@n38, False=@n39) {defaultValue=true, parameterName=UseWaterDisplacement}
   ))
  n40 = SphereMask(A=TextureCoordinate(), B=Constant() {r=0.5}, Radius=Constant() {r=0.5}, Hardness=Constant() {r=0.5}) {attenuationRadius=0.5, hardnessPercent=0}
  n41 = Power(Base=NamedRerouteUsage(), Exp=ScalarParameter() {defaultValue=1, parameterName=CircleMaskPower})
  n42 = ScalarParameter() {defaultValue=1, parameterName=CircleMaskMultiply}
  n43 = Constant3Vector() {constant=<rgba 0.16 0.4 0.4 1>}
  n44 = Multiply(A=@n2.R)
  n45 = Min(A=OneMinus(None=@n44), B=@n44)
  n46 = Constant() {r=50}
  n47 = Power(Base=Saturate(None=Multiply(A=@n45, B=@n46)))
  n48 = VectorParameter() {group=Move, parameterName=Move6, defaultValue=<rgba 0.35 0.25 0 -0.06>}
  n49 = Multiply(A=@n1.Radial Coordinates, B=AppendVector(A=@n48.R, B=@n48.R))
  n50 = MaterialFunctionCall(Float2=ActorPositionWS()) {materialFunction=&BreakOutFloat2Components}
  n51 = Constant() {r=100}
  n52 = LinearInterpolate(B=AppendVector(A=Constant(), B=Fmod(A=@n50.G, B=@n51)), Alpha=ScalarParameter() {defaultValue=1, parameterName=UseRandomUVCoordinate})
  n53 = Add(
   A=Multiply(A=ComponentMask(None=@n8.RGB) {g=true, r=true}, B=Constant() {r=0.005})
   B=Panner(Coordinate=Add(A=@n49, B=@n52), Time=@n7, Speed=AppendVector(A=@n48.B, B=@n48.B)) {speedY=-0.5}
  )
  n54 = TextureSample!(UVs=@n53, Tex, Apply View MipBias) {samplerType=SAMPLERTYPE_Masks, texture=/FortWater/Textures/T_OceanFoam_New_Mask}
  n55 = Power(Base=@n54.R, Exp=Constant() {r=2})
  n56 = Multiply(A=@n54.R, B=Constant() {r=0.5})
  n57 = MaterialFunctionCall(Float2=@n1.Radial Coordinates) {materialFunction=&BreakOutFloat2Components}
  n58 = VectorParameter() {group=foam, parameterName=FoamMask, defaultValue=<rgba 0.55 0.25 0.25 5>}
  n59 = Divide(A=Subtract(A=OneMinus(None=@n57.G), B=@n58.R), B=OneMinus(None=@n58.R))
  n60 = Multiply(A=Divide(A=Subtract(A=@n57.G, B=@n58.G), B=OneMinus(None=@n58.G)), B=@n58.A)
  n61 = Saturate(None=Multiply(A=Add(A=@n56, B=@n59), B=Saturate(None=@n60)))
  n62 = Multiply(A=@n55, B=@n61)
  n63 = Subtract(A=Distance(A=CameraPositionWS(), B=ActorPositionWS())) {constB=2000}
  n64 = LinearInterpolate(A=Constant() {r=1.2}, B=Constant() {r=0.5}, Alpha=Saturate(None=Divide(A=@n63) {constB=1000}))
  n65 = Multiply(A=Multiply(A=@n47, B=Multiply(A=@n62, B=@n64)))
  n66 = Constant3Vector() {constant=<rgba 0.6 1.7667 2 1>}
  n67 = SphereMask(A=TextureCoordinate(), B=Constant() {r=0.5}) {attenuationRadius=0.5, hardnessPercent=0}
  n68 = Power(Base=@n29.RGB, Exp=Constant() {r=2})
  n69 = SphereMask(A=TextureCoordinate(), B=Constant() {r=0.5}, Radius=Constant() {r=0.5}) {attenuationRadius=0.5, hardnessPercent=0}
  n70 = TextureSample!(UVs=@n25, Tex, "DDX(UVs)"=@n27, "DDY(UVs)"=@n28, Apply View MipBias) {
   mipValueMode=TMVM_Derivative
   samplerType=SAMPLERTYPE_Normal
   texture=&/T_FlowSandTile_N
  }
  n71 = VectorParameter() {group=Move, parameterName=Move5, defaultValue=<rgba 1 1 0 -0.4>}
  n72 = AppendVector(A=@n71.R, B=@n71.R)
  n73 = Add(
   A=Multiply(A=AppendVector(A=@n70.R, B=@n70.R), B=Constant() {r=0.01})
   B=Panner(Coordinate=Multiply(A=@n1.Radial Coordinates, B=@n72), Time=@n7, Speed=AppendVector(A=@n71.B, B=@n71.B)) {speedY=-0.5}
  )
  n74 = Multiply(
   A=Power(Base=OneMinus(None=@n69), Exp=Constant() {r=2})
   B=Power(
    Base=TextureSample!(UVs=@n73, Tex, Apply View MipBias) {samplerType=SAMPLERTYPE_Masks, texture=&T_OceanFoam_New_Mask}.R
    Exp=Constant() {r=2}
   )
  )
  n75 = Multiply(A=LinearInterpolate(B=@n67, Alpha=Multiply(A=@n68, B=@n74)), B=@n64)
  n76 = MaterialFunctionCall(
   Normal=TextureSample(UVs=@n53) {samplerType=SAMPLERTYPE_Normal, texture=/FortWater/Textures/T_OceanFoam_New_N}.RGB
   Flatness=Constant() {r=0.4}
  ) {materialFunction=&FlattenNormal}
  n77 = Transform(None=@n76.Result)
  n78 = Multiply(A=Constant3Vector() {constant=<rgba 0.5 0.7521 1 0.5>}, B=Constant() {r=5})
  n79 = ShadingPathSwitch(Default=SkyLightEnvMapSample(Direction=ReflectionVectorWS(CustomWorldNormal=@n77)), Forward=@n78, Mobile=@n78)
  n80 = Multiply(A=@n55, B=Multiply(A=@n61, B=QualitySwitch(Default=@n79, Low=@n78)))
  n81 = Multiply(
   A=Multiply(A=@n47, B=Multiply(A=@n80, B=@n64))
   B=VectorParameter() {group=foam, parameterName=FoamEmissiveColor, defaultValue=<rgba 0.27 0.27 0.27 1>}.RGB
  )
  n82 = Add(A=Add(A=Multiply(A=@n43, B=@n65), B=Multiply(A=@n66, B=@n75)), B=@n81)
  n83 = Desaturation(None=@n82, Fraction=Constant() {r=1})
  n84 = QualitySwitch(Default=ShadingPathSwitch(Default=Constant() {r=1}, Forward=@n83, Mobile=@n83), Low=@n83)
  n85 = Multiply(A=Saturate(None=Multiply(A=@n41, B=@n42)), B=@n84)
  n86 = Multiply(A=NamedRerouteUsage(), B=Add(A=NamedRerouteUsage(), B=StaticSwitchParameter(True=@n85, False=@n84) {parameterName=CircleMask}))
  n87 = LinearInterpolate(A=@n86, B=Multiply(A=@n86, B=ParticleColor().A), Alpha=ScalarParameter() {parameterName=UseParticleColorAlphaInput})
  n88 = TextureCoordinate()
  n89 = Constant() {r=0.5}
  n90 = Constant() {r=0.5}
  n91 = Constant()
  n92 = LinearInterpolate(
   A=@n87
   B=Multiply(A=@n87, B=SphereMask(A=@n88, B=@n89, Radius=@n90, Hardness=@n91) {attenuationRadius=0.5, hardnessPercent=0})
   Alpha=ScalarParameter() {parameterName=UseOpacityFalloffMask}
  )
  Reroute(None=DepthFade(Opacity=Multiply(A=@n40, B=@n92)) {fadeDistanceDefault=2})
  n93 = VectorParameter() {
   desc="Custom color set by BP, uses WaterBodyPrimaryColor"
   group=HSV Offset
   parameterName=WaterBodyPrimaryColor
   defaultValue=<rgba 0.25 1 1 0>
  }
  n94 = Add(A=@n82, B=NamedRerouteUsage())
  n95 = Normalize(VectorInput=@n94)
  n96 = QualitySwitch(Default=ShadingPathSwitch(Default=@n94, Forward=@n95, Mobile=@n95), Low=@n95)
  n97 = Multiply(A=ComponentMask(None=@n96) {r=true}, B=@n93.A)
  n98 = ScalarParameter() {defaultValue=1, parameterName=HSVOffset Mask Power}
  n99 = Saturate(None=Multiply(A=Power(Base=@n97, Exp=@n98) {constExponent=0.5}, B=ScalarParameter() {defaultValue=5, parameterName=HSVOffset Mask Scale}) {constB=8})
  Reroute(None=LinearInterpolate(A=@n96, B=@n93.RGB, Alpha=@n99))
  n100 = NamedRerouteUsage()
  n101 = TextureCoordinate()
  n102 = Constant() {r=0.5}
  n103 = Constant() {r=0.4}
  n104 = Constant()
  n105 = LinearInterpolate(
   A=LinearInterpolate(B=Constant() {r=1.2}, Alpha=SphereMask(A=@n101, B=@n102, Radius=@n103, Hardness=@n104) {attenuationRadius=0.5, hardnessPercent=0}) {constA=1}
   Alpha=SphereMask(A=TextureCoordinate(), B=Constant() {r=0.5}, Radius=Constant() {r=0.2}, Hardness=Constant() {r=0.2}) {attenuationRadius=0.5, hardnessPercent=0}
  ) {constA=1}
  n106 = Constant() {r=100}
  n107 = LinearInterpolate(
   A=LinearInterpolate(B=Constant() {r=2}, Alpha=@n67) {constA=1}
   Alpha=SphereMask(A=TextureCoordinate(), B=Constant() {r=0.5}, Radius=Constant() {r=0.2}, Hardness=Constant() {r=0.4}) {attenuationRadius=0.5, hardnessPercent=0}
  ) {constA=1}
  n108 = Constant()
  n109 = SceneColor!(OffsetFraction=Add(
    A=MaterialFunctionCall!(RefractionStrength=@n105, DepthBias=@n106, FalloffDistance, NormalTS=@n8.RGB) {materialFunction=/Game/Effects/Niagara/MasterMaterials/Functions/Common/FXCommon_PixelNormalOffsetRefraction}.ScreenUVOffsetFraction
    B=MaterialFunctionCall(RefractionStrength=@n107, DepthBias=@n108, NormalTS=@n70.RGB) {materialFunction=&FXCommon_PixelNormalOffsetRefraction}.ScreenUVOffsetFraction
   )) {inputMode=OffsetFraction}
  n110 = Constant()
  n111 = LinearInterpolate(A=QualitySwitch(Default=ShadingPathSwitch(Default=@n109.RGB, Mobile=@n110), Low=@n110), B=Constant(), Alpha=@n65)
  n112 = Subtract(A=DotProduct(A=SkyAtmosphereLightDirection(), B=@n76.Result), B=Constant() {r=0.8})
  n113 = Constant()
  n114 = ShadingPathSwitch(Default=Multiply(A=Saturate(None=@n112), B=Constant() {r=1.2}), Forward=@n113, Mobile=@n113)
  n115 = Multiply(A=@n64, B=Multiply(A=@n61, B=QualitySwitch(Default=@n114, Low=@n113)))
  n116 = Add(A=NamedRerouteUsage(), B=Add(A=Add(A=@n111, B=@n81), B=Multiply(A=@n47, B=@n115)))
  n117 = MaterialFunctionCall(Emissive=Divide(A=@n116)) {materialFunction=&MF_FX_EmissiveAdjustments}
  n118 = QualitySwitch(Default=ShadingPathSwitch(Default=@n116, Forward=@n117.Result, Mobile=@n117.Result), Low=@n117.Result)
  Reroute(None=LinearInterpolate(A=Add(A=@n100, B=@n118), B=@n93.RGB, Alpha=@n99))
 ]
}
